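# Player animation layer Blueprint for riding the Labrador in a hard saddle, overriding vehicle full-body and final poses and reusing some Wolf assets.
# A locomotion state machine with corrective leans drives ride mode.
# Outside ride mode, the incoming upper body is layered over it with combat additives, yaw aim offset and melee twist, then inertialized.
# The final-pose layer offsets and counter-rotates the root in world space.
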AnimBlueprint /LabradorRidingContent/Animation/HardSaddle/Labrador/LabradorRiding_LabradorPlayerLayerAnimBP
target_skeleton=&Fortnite_M_Avg_Player_Skeleton
graphs [
 AnimGraph {nodes [Output Pose() {layerGroup=None}]}
 CorrectiveLeanPoses {nodes [
   n1 = Blend Poses by bool(
    BlendPose_0="Sequence Player (additive) 'LabradorRiding_Player_Leans_Right'"() {bLoopAnimation=false, sequence=&/LabradorRiding_Player_Leans_Right}
    BlendPose_1="Sequence Player (additive) 'LabradorRiding_Player_Leans_Left'"() {bLoopAnimation=false, sequence=&/LabradorRiding_Player_Leans_Left}
    BlendTime_0=0.4
    BlendTime_1=0.4
    bActiveValue=Get bIsLeaningRight() -> (bIsLeaningRight:bool)
   )
   n2 = Use cached pose 'PreCorrectiveLeanPose'()
   n3 = Get AlphaValue() -> (AlphaValue:float)
   Output Pose(Result=Two Way Blend(
     A=Apply Mesh Space Additive!(Base:Pose Link Structure=@n2, Additive:Pose Link Structure=@n1, Alpha:float=@n3) -> (Pose:Pose Link Structure) {is=title AnimGraphNode_ApplyMeshSpaceAdditive}
     B=Use cached pose 'PreCorrectiveLeanPose'()
     Alpha=Get DisableLeanWeight() -> (DisableLeanWeight:double)
    ))
   Save cached pose 'PreCorrectiveLeanPose'(Pose=Input Pose() {name=InPose_1}) {cacheName=PreCorrectiveLeanPose}
   Use cached pose 'PreCorrectiveLeanPose'()
  ]}
 EventGraph {nodes [
   Event PlaceHandIKTargets() {is=title K2Node_Event}
   Event BlueprintUpdateAnimation()
  ]}
 VehicleFinalPoseOverride {nodes [
   Output Pose(Result=Use cached pose 'PreLeans'()) {layerGroup=VehicleOverride}
   n1 = "Transform (Modify) Bone - Bone: Root"!(
    ComponentPose=Local To Component(LocalPose=Input Pose() {name=InFinalPose})
    Alpha=1
    Translation=Get PlayerOffset() -> (PlayerOffset:Vector)
   ) -> (Pose) {
    rotationSpace=BCS_WorldSpace
    translationMode=BMM_Additive
    translationSpace=BCS_WorldSpace
    alphaBoolBlend {blendOption=Linear}
    boneToModify {boneName=Root}
   }
   n2 = "Transform (Modify) Bone - Bone: Root"!(
    ComponentPose=@n1
    Alpha=Get BodyCounterRotationAlpha() -> (BodyCounterRotationAlpha:float)
    Rotation:Rotator=Get BodyCounterRotation() -> (BodyCounterRotation:Rotator)
   ) -> (Pose) {
    rotationMode=BMM_Replace
    rotationSpace=BCS_WorldSpace
    translationSpace=BCS_WorldSpace
    alphaBoolBlend {blendOption=Linear}
    boneToModify {boneName=Root}
   }
   Save cached pose 'PreLeans'(Pose=Component To Local(ComponentPose=@n2)) {cacheName=PreLeans}
  ]}
 VehicleFullBodyOverride {nodes [
   n1 = Layered blend per bone(BasePose=Use cached pose 'PostUpperBodyPose'(), BlendPoses_0=Use cached pose 'UpperAndLowerPose'()) {
    bBlendRootMotionBasedOnRootBone=false
    bMeshSpaceRotationBlend=true
    layerSetup [{branchFilters [
       {boneName=Root}
       {blendDepth=-1, boneName=pelvis}
       {boneName=Spine_01}
       {blendDepth=-1, boneName=thigh_l}
       {blendDepth=-1, boneName=thigh_r}
       {blendDepth=-1, boneName=ik_foot_root}
      ]}]
   }
   AnimGraphNode_StateMachine_2 = CombatUpperBodyAdditive() {is=title AnimGraphNode_StateMachine, bAllowConduitEntryStates=true, bSkipFirstUpdateTransition=false, maxTransitionsPerFrame=1}
   n2 = Get NoiseAlpha() -> (NoiseAlpha:float)
   AnimGraphNode_StateMachine_4 = EntryAdditive() {is=title AnimGraphNode_StateMachine}
   n3 = AimOffset Player 'LabradorRiding_HardSaddleWolf_Player_AO_Yaw'!(
    X:float=Get BodyYawOffset() -> (BodyYawOffset:float)
    BasePose:Pose Link Structure=Apply Mesh Space Additive(
     Base=Apply Mesh Space Additive(Base=@n1, Additive=@AnimGraphNode_StateMachine_2, Alpha=@n2) {alphaBoolBlend {blendInTime=0.6, blendOutTime=0.35}}
     Additive=@AnimGraphNode_StateMachine_4
     Alpha=1
    ) {alphaBoolBlend {blendInTime=0.6, blendOutTime=0.35}}
    Alpha:float=1
   ) -> (Pose:Pose Link Structure) {
    blendSpace=/LabradorRidingContent/Animation/HardSaddle/Wolf/AO/LabradorRiding_HardSaddleWolf_Player_AO_Yaw
    alphaBoolBlend {blendOption=Linear}
   }
   AnimGraphNode_StateMachine_3 = MeleeOffset() {is=title AnimGraphNode_StateMachine}
   n4 = Blend Poses by bool(
    BlendPose_0=Use cached pose 'PostUpperBodyPose'()
    BlendPose_1=Apply Additive!(
     Base=@n3
     Additive=@AnimGraphNode_StateMachine_3
     bAlphaBoolEnabled:bool=Get bShouldAdjustBodyTwistForMelee() -> (bShouldAdjustBodyTwistForMelee:bool)
    ) -> (Pose) {alphaInputType=Bool, alphaBoolBlend {blendInTime=0.25, blendOption=Linear, blendOutTime=0.25}}
    BlendTime_0=0.25
    BlendTime_1=Get FalseBlendTime() -> (FalseBlendTime:float)
    bActiveValue=Get bIsRideMode() -> (bIsRideMode:bool)
   ) {blendProfile=&Fortnite_M_Avg_Player_Skeleton:FastWeapons, blendType=Linear}
   Output Pose(Result=Inertialization!(Source:Pose Link Structure=WeaponAttachmentProceduralAdjustment!(InPose:Pose Link Structure=@n4) -> (Pose:Pose Link Structure) {is=title AnimGraphNode_LinkedAnimLayer, layer=WeaponAttachmentProceduralAdjustment}) -> (Pose:Pose Link Structure) {is=title AnimGraphNode_Inertialization}) {layerGroup=VehicleOverride}
   Save cached pose 'UpperAndLowerPose'(Pose=Input Pose() {name=InPoseFullBody}) {cacheName=UpperAndLowerPose}
   n5 = Use cached pose 'FullBodyPose'()
   AnimGraphNode_StateMachine_1 = UpperBody() {is=title AnimGraphNode_StateMachine, bAllowConduitEntryStates=true, bSkipFirstUpdateTransition=false, maxTransitionsPerFrame=1}
   n6 = Layered blend per bone(
    BasePose=Blend Poses by bool(
     BlendPose_0=CorrectiveLeanPoses!(InPose_1:Pose Link Structure=@n5) -> (Pose:Pose Link Structure) {is=title AnimGraphNode_LinkedAnimLayer, layer=CorrectiveLeanPoses}
     BlendPose_1=Use cached pose 'FullBodyPose'()
     BlendTime_0=0.4
     BlendTime_1=0.4
     bActiveValue=true
    )
    BlendPoses_0=@AnimGraphNode_StateMachine_1
   ) {
    bBlendRootMotionBasedOnRootBone=false
    bMeshSpaceRotationBlend=true
    layerSetup [{branchFilters [
       {boneName=Root}
       {blendDepth=-1, boneName=pelvis}
       {boneName=Spine_01}
       {blendDepth=-1, boneName=thigh_l}
       {blendDepth=-1, boneName=thigh_r}
       {blendDepth=-1, boneName=ik_foot_root}
      ]}]
   }
   Save cached pose 'PostUpperBodyPose'(Pose=@n6) {cacheName=PostUpperBodyPose}
   AnimGraphNode_StateMachine_0 = FullBody() {is=title AnimGraphNode_StateMachine, bAllowConduitEntryStates=true, bSkipFirstUpdateTransition=false}
   Save cached pose 'FullBodyPose'(Pose=@AnimGraphNode_StateMachine_0) {cacheName=FullBodyPose}
  ]}
 FullBody {
  entry=Idle
  path=VehicleFullBodyOverride.AnimGraphNode_StateMachine_0.FullBody
  ids=id
  nodes [
   SA_ToJumpStart() {is=title AnimStateAliasNode, also {aliases [Idle,Moving,Startle,Swimming]}}
   SA_ToSwimming() {is=title AnimStateAliasNode, bGlobalAlias=true}
   AnimStateNode_0 = JumpStart() {is=title AnimStateNode}
   AnimStateNode_10 = Startle() {is=title AnimStateNode}
   AnimStateNode_2 = Idle() {
    stateEntered {guid=#1}
    stateFullyBlended {guid=#2}
    stateLeft {guid=#3}
   }
   AnimStateNode_3 = Moving() {
    is=title AnimStateNode
    stateEntered {guid=#4}
    stateFullyBlended {guid=#5}
    stateLeft {guid=#6}
   }
   AnimStateNode_4 = JumpFall() {is=title AnimStateNode}
   AnimStateNode_6 = JumpLand() {is=title AnimStateNode}
   AnimStateNode_7 = Swimming() {is=title AnimStateNode}
   AnimStateTransitionNode_0 = AnimStateTransitionNode() {logicType=TLT_Inertialization}
   AnimStateTransitionNode_1 = AnimStateTransitionNode() {bAllowInertializationForSelfTransitions=true, priorityOrder=2}
   AnimStateTransitionNode_10 = AnimStateTransitionNode() {logicType=TLT_Inertialization, priorityOrder=2}
   AnimStateTransitionNode_11 = AnimStateTransitionNode() {bAllowInertializationForSelfTransitions=true}
   AnimStateTransitionNode_24 = AnimStateTransitionNode() {bAllowInertializationForSelfTransitions=true, bAutomaticRuleBasedOnSequencePlayerInState=true}
   AnimStateTransitionNode_3 = AnimStateTransitionNode() {bAllowInertializationForSelfTransitions=true}
   AnimStateTransitionNode_4 = AnimStateTransitionNode() {logicType=TLT_Inertialization, priorityOrder=100}
   AnimStateTransitionNode_5 = AnimStateTransitionNode() {bAutomaticRuleBasedOnSequencePlayerInState=true, crossfadeDuration=0.3, logicType=TLT_Inertialization, priorityOrder=100}
   AnimStateTransitionNode_6 = AnimStateTransitionNode() {crossfadeDuration=0.4, logicType=TLT_Inertialization}
   AnimStateTransitionNode_7 = AnimStateTransitionNode() {logicType=TLT_Inertialization}
   AnimStateTransitionNode_8 = AnimStateTransitionNode() {bAllowInertializationForSelfTransitions=true}
  ]
  links [
   {
    dst=Moving
    rule=VehicleFullBodyOverride.AnimGraphNode_StateMachine_0.FullBody.AnimStateTransitionNode_8.Transition
    src=Idle
    via=AnimStateTransitionNode_8
   }
   {
    dst=Startle
    rule=VehicleFullBodyOverride.AnimGraphNode_StateMachine_0.FullBody.AnimStateTransitionNode_3.Transition
    src=Idle
    via=AnimStateTransitionNode_3
   }
   {
    dst=JumpLand
    rule=VehicleFullBodyOverride.AnimGraphNode_StateMachine_0.FullBody.AnimStateTransitionNode_4.Transition
    src=JumpFall
    via=AnimStateTransitionNode_4
   }
   {
    dst=Idle
    rule=VehicleFullBodyOverride.AnimGraphNode_StateMachine_0.FullBody.AnimStateTransitionNode_5.Transition
    src=JumpLand
    via=AnimStateTransitionNode_5
   }
   {
    dst=Moving
    rule=VehicleFullBodyOverride.AnimGraphNode_StateMachine_0.FullBody.AnimStateTransitionNode_10.Transition
    src=JumpLand
    via=AnimStateTransitionNode_10
   }
   {
    dst=JumpFall
    rule=VehicleFullBodyOverride.AnimGraphNode_StateMachine_0.FullBody.AnimStateTransitionNode_11.Transition
    src=JumpStart
    via=AnimStateTransitionNode_11
   }
   {
    dst=JumpFall
    rule=VehicleFullBodyOverride.AnimGraphNode_StateMachine_0.FullBody.AnimStateTransitionNode_1.Transition
    src=SA_ToJumpStart
    via=AnimStateTransitionNode_1
   }
   {
    dst=JumpStart
    rule=VehicleFullBodyOverride.AnimGraphNode_StateMachine_0.FullBody.AnimStateTransitionNode_0.Transition
    src=SA_ToJumpStart
    via=AnimStateTransitionNode_0
   }
   {
    dst=Swimming
    rule=VehicleFullBodyOverride.AnimGraphNode_StateMachine_0.FullBody.AnimStateTransitionNode_6.Transition
    src=SA_ToSwimming
    via=AnimStateTransitionNode_6
   }
   {
    dst=Moving
    rule=VehicleFullBodyOverride.AnimGraphNode_StateMachine_0.FullBody.AnimStateTransitionNode_24.AnimationTransitionGraph_0
    src=Startle
    via=AnimStateTransitionNode_24
   }
   {
    dst=Moving
    rule=VehicleFullBodyOverride.AnimGraphNode_StateMachine_0.FullBody.AnimStateTransitionNode_7.Transition
    src=Swimming
    via=AnimStateTransitionNode_7
   }
  ]
 }
 JumpStart {
  path=VehicleFullBodyOverride.AnimGraphNode_StateMachine_0.FullBody.AnimStateNode_0.JumpStart
  nodes [Output Animation Pose(Result=Sequence Player 'LabradorRiding_Player_Jump_Start'() {bLoopAnimation=false, sequence=&/LabradorRiding_Player_Jump_Start})]
 }
 Startle {
  path=VehicleFullBodyOverride.AnimGraphNode_StateMachine_0.FullBody.AnimStateNode_10.Startle
  nodes [Output Animation Pose(Result=Sequence Player 'LabradorRiding_Player_Startle'() {bLoopAnimation=false, sequence=&/LabradorRiding_Player_Startle})]
 }
 Idle {
  path=VehicleFullBodyOverride.AnimGraphNode_StateMachine_0.FullBody.AnimStateNode_2.Idle
  nodes [Output Animation Pose(Result=Sequence Player 'LabradorRiding_Player_Idle'() {sequence=&/LabradorRiding_Player_Idle})]
 }
 Moving {
  path=VehicleFullBodyOverride.AnimGraphNode_StateMachine_0.FullBody.AnimStateNode_3.Moving
  nodes [
   n1 = "Sequence Evaluator (by time)"!() -> (Pose:Pose Link Structure) {
    is=title AnimGraphNode_SequenceEvaluator
    node {updateFunction {functionName=ANF_Movement_Update}}
    updateFunction {bSelfContext=true, memberName=ANF_Movement_Update}
   }
   Output Animation Pose(Result=@n1)
  ]
 }
 JumpFall {
  path=VehicleFullBodyOverride.AnimGraphNode_StateMachine_0.FullBody.AnimStateNode_4.JumpFall
  nodes [
   n1 = Blendspace Player 'LabradorRiding_HardSaddleWolf_Player_SlopeSliding_BS'(X=Get SlopeSlidingRoll() -> (SlopeSlidingRoll:float), Y=Get SlopeSlidingPitch() -> (SlopeSlidingPitch:float)) {blendSpace=/LabradorRidingContent/Animation/HardSaddle/Wolf/SlopeSliding/LabradorRiding_HardSaddleWolf_Player_SlopeSliding_BS}
   n2 = "Sequence Player (additive) 'LabradorRiding_HardSaddleWolf_Player_SlopeSliding_Add'"() {sequence=/LabradorRidingContent/Animation/HardSaddle/Wolf/SlopeSliding/LabradorRiding_HardSaddleWolf_Player_SlopeSliding_Add}
   Output Animation Pose(Result=Blend Poses by bool(
     BlendPose_0=Apply Additive!(Base=@n1, Additive=@n2, Alpha:float) -> (Pose) {alphaBoolBlend {blendOption=Linear}}
     BlendPose_1=Sequence Player 'LabradorRiding_Player_Jump_FallLoop'() {sequence=&/LabradorRiding_Player_Jump_FallLoop}
     bActiveValue=Get bIsSlopeSliding() -> (bIsSlopeSliding:bool)
    ) {blendType=Linear})
  ]
 }
 JumpLand {
  path=VehicleFullBodyOverride.AnimGraphNode_StateMachine_0.FullBody.AnimStateNode_6.JumpLand
  nodes [Output Animation Pose(Result=Sequence Player 'LabradorRiding_Player_Jump_Land'() {bLoopAnimation=false, sequence=&/LabradorRiding_Player_Jump_Land})]
 }
 Swimming {
  path=VehicleFullBodyOverride.AnimGraphNode_StateMachine_0.FullBody.AnimStateNode_7.Swimming
  nodes [Output Animation Pose(Result=Sequence Player 'LabradorRiding_Player_Swim_N'() {bLoopAnimation=false, sequence=&/LabradorRiding_Player_Swim_N})]
 }
 Transition {
  path=VehicleFullBodyOverride.AnimGraphNode_StateMachine_0.FullBody.AnimStateTransitionNode_0.Transition
  nodes [Result(bCanEnterTransition=Get bIsJumping_BP() -> (bIsJumping_BP:bool))]
 }
 Transition {
  path=VehicleFullBodyOverride.AnimGraphNode_StateMachine_0.FullBody.AnimStateTransitionNode_1.Transition
  nodes [Result(bCanEnterTransition=Get bIsFalling_BP() -> (bIsFalling_BP:bool))]
 }
 Transition {
  path=VehicleFullBodyOverride.AnimGraphNode_StateMachine_0.FullBody.AnimStateTransitionNode_10.Transition
  nodes [Result(bCanEnterTransition=Get bLandToMovement() -> (bLandToMovement:bool))]
 }
 Transition {
  path=VehicleFullBodyOverride.AnimGraphNode_StateMachine_0.FullBody.AnimStateTransitionNode_11.Transition
  nodes [Result(bCanEnterTransition=Get bIsFalling_BP())]
 }
 AnimationTransitionGraph_0 {path=VehicleFullBodyOverride.AnimGraphNode_StateMachine_0.FullBody.AnimStateTransitionNode_24.AnimationTransitionGraph_0, nodes [Result()]}
 Transition {
  path=VehicleFullBodyOverride.AnimGraphNode_StateMachine_0.FullBody.AnimStateTransitionNode_3.Transition
  nodes [Result(bCanEnterTransition=Get bIdleToStartled() -> (bIdleToStartled:bool))]
 }
 Transition {
  path=VehicleFullBodyOverride.AnimGraphNode_StateMachine_0.FullBody.AnimStateTransitionNode_4.Transition
  nodes [Result(bCanEnterTransition=NOT Boolean(A=Get bIsFalling_BP()))]
 }
 Transition {path=VehicleFullBodyOverride.AnimGraphNode_StateMachine_0.FullBody.AnimStateTransitionNode_5.Transition, nodes [Result()]}
 Transition {
  path=VehicleFullBodyOverride.AnimGraphNode_StateMachine_0.FullBody.AnimStateTransitionNode_6.Transition
  nodes [Result(bCanEnterTransition=Get bIsSurfaceSwimming())]
 }
 Transition {
  path=VehicleFullBodyOverride.AnimGraphNode_StateMachine_0.FullBody.AnimStateTransitionNode_7.Transition
  nodes [Result(bCanEnterTransition=NOT Boolean(A=Get bIsSurfaceSwimming()))]
 }
 Transition {
  path=VehicleFullBodyOverride.AnimGraphNode_StateMachine_0.FullBody.AnimStateTransitionNode_8.Transition
  nodes [Result(bCanEnterTransition=Get bIdleToMovement() -> (bIdleToMovement:bool))]
 }
 UpperBody {
  entry=FullBodyPose
  path=VehicleFullBodyOverride.AnimGraphNode_StateMachine_1.UpperBody
  ids=id
  nodes [
   AnimStateNode_0 = HackFullBodyPose() {is=title AnimStateNode}
   AnimStateNode_1 = CombatToRide() {is=title AnimStateNode}
   AnimStateNode_2 = FullBodyPose() {is=title AnimStateNode}
   AnimStateTransitionNode_5 = AnimStateTransitionNode() {bDisabled=true, crossfadeDuration=0.1, logicType=TLT_Inertialization}
   AnimStateTransitionNode_6 = AnimStateTransitionNode() {blendMode=Linear, crossfadeDuration=0, logicType=TLT_Inertialization}
   AnimStateTransitionNode_8 = AnimStateTransitionNode() {bAutomaticRuleBasedOnSequencePlayerInState=true, logicType=TLT_Inertialization}
  ]
  links [
   {
    dst=HackFullBodyPose
    rule=VehicleFullBodyOverride.AnimGraphNode_StateMachine_1.UpperBody.AnimStateTransitionNode_8.Transition
    src=CombatToRide
    via=AnimStateTransitionNode_8
   }
   {
    dst=CombatToRide
    rule=VehicleFullBodyOverride.AnimGraphNode_StateMachine_1.UpperBody.AnimStateTransitionNode_5.Transition
    src=FullBodyPose
    via=AnimStateTransitionNode_5
   }
   {
    dst=FullBodyPose
    rule=VehicleFullBodyOverride.AnimGraphNode_StateMachine_1.UpperBody.AnimStateTransitionNode_6.Transition
    src=HackFullBodyPose
    via=AnimStateTransitionNode_6
   }
  ]
 }
 HackFullBodyPose {
  path=VehicleFullBodyOverride.AnimGraphNode_StateMachine_1.UpperBody.AnimStateNode_0.HackFullBodyPose
  nodes [Output Animation Pose(Result=Use cached pose 'FullBodyPose'())]
 }
 CombatToRide {
  path=VehicleFullBodyOverride.AnimGraphNode_StateMachine_1.UpperBody.AnimStateNode_1.CombatToRide
  nodes [Output Animation Pose(Result=Blendspace Player 'LabradorRiding_HardSaddleWolf_Player_CtR_BS'!(X=Get CombatToRideYaw() -> (CombatToRideYaw:float)) -> (Pose) {bLoop=false, blendSpace=/LabradorRidingContent/Animation/HardSaddle/Wolf/LabradorRiding_HardSaddleWolf_Player_CtR_BS})]
 }
 FullBodyPose {
  path=VehicleFullBodyOverride.AnimGraphNode_StateMachine_1.UpperBody.AnimStateNode_2.FullBodyPose
  nodes [Output Animation Pose(Result=Use cached pose 'FullBodyPose'())]
 }
 Transition {
  path=VehicleFullBodyOverride.AnimGraphNode_StateMachine_1.UpperBody.AnimStateTransitionNode_5.Transition
  nodes [Result(bCanEnterTransition=Get bIsRideMode())]
 }
 Transition {
  path=VehicleFullBodyOverride.AnimGraphNode_StateMachine_1.UpperBody.AnimStateTransitionNode_6.Transition
  nodes [Result(bCanEnterTransition=NOT Boolean(A=Get bIsRideMode()))]
 }
 Transition {path=VehicleFullBodyOverride.AnimGraphNode_StateMachine_1.UpperBody.AnimStateTransitionNode_8.Transition, nodes [Result()]}
 CombatUpperBodyAdditive {
  entry=Idle
  path=VehicleFullBodyOverride.AnimGraphNode_StateMachine_2.CombatUpperBodyAdditive
  ids=id
  nodes [
   SA_ToSwimming() {bGlobalAlias=true}
   Jump Alias() {is=title AnimStateAliasNode, also {aliases [Idle,JumpLand,Moving,Stop]}}
   SA_ToMoving() {is=title AnimStateAliasNode, also {aliases [Idle,JumpLand,Stop]}}
   AnimStateNode_0 = Moving() {
    stateEntered {guid=#7}
    stateFullyBlended {guid=#8}
    stateLeft {guid=#9}
   }
   AnimStateNode_2 = Swimming()
   AnimStateNode_4 = Idle()
   AnimStateNode_5 = JumpStart()
   AnimStateNode_6 = JumpFall()
   AnimStateNode_7 = Stop() {is=title AnimStateNode}
   AnimStateNode_8 = JumpLand()
   AnimStateTransitionNode_1 = AnimStateTransitionNode() {bAutomaticRuleBasedOnSequencePlayerInState=true, crossfadeDuration=0.3, logicType=TLT_Inertialization, priorityOrder=100}
   AnimStateTransitionNode_10 = AnimStateTransitionNode() {crossfadeDuration=0.35, logicType=TLT_Inertialization, priorityOrder=3}
   AnimStateTransitionNode_12 = AnimStateTransitionNode() {blendMode=Linear, crossfadeDuration=0.3, priorityOrder=2}
   AnimStateTransitionNode_17 = AnimStateTransitionNode() {bAllowInertializationForSelfTransitions=true}
   AnimStateTransitionNode_18 = AnimStateTransitionNode() {bAutomaticRuleBasedOnSequencePlayerInState=true, crossfadeDuration=0.4, priorityOrder=3}
   AnimStateTransitionNode_2 = AnimStateTransitionNode() {
    crossfadeDuration=0.6
    logicType=TLT_Inertialization
    priorityOrder=2
    transitionEnd {guid=#10}
    transitionInterrupt {guid=#11}
    transitionStart {guid=#12}
   }
   AnimStateTransitionNode_3 = AnimStateTransitionNode() {crossfadeDuration=0.4, logicType=TLT_Inertialization}
   AnimStateTransitionNode_4 = AnimStateTransitionNode() {crossfadeDuration=0.3, logicType=TLT_Inertialization, priorityOrder=100}
   AnimStateTransitionNode_5 = AnimStateTransitionNode() {logicType=TLT_Inertialization}
   AnimStateTransitionNode_6 = AnimStateTransitionNode() {logicType=TLT_Inertialization, priorityOrder=100}
   AnimStateTransitionNode_7 = AnimStateTransitionNode() {bAutomaticRuleBasedOnSequencePlayerInState=true, crossfadeDuration=0.4, logicType=TLT_Inertialization}
   AnimStateTransitionNode_9 = AnimStateTransitionNode() {logicType=TLT_Inertialization, priorityOrder=2}
  ]
  links [
   {
    dst=JumpFall
    rule=VehicleFullBodyOverride.AnimGraphNode_StateMachine_2.CombatUpperBodyAdditive.AnimStateTransitionNode_17.Transition
    src=Jump Alias
    via=AnimStateTransitionNode_17
   }
   {
    dst=JumpStart
    rule=VehicleFullBodyOverride.AnimGraphNode_StateMachine_2.CombatUpperBodyAdditive.AnimStateTransitionNode_4.Transition
    src=Jump Alias
    via=AnimStateTransitionNode_4
   }
   {
    dst=JumpLand
    rule=VehicleFullBodyOverride.AnimGraphNode_StateMachine_2.CombatUpperBodyAdditive.AnimStateTransitionNode_6.Transition
    src=JumpFall
    via=AnimStateTransitionNode_6
   }
   {
    dst=Idle
    rule=VehicleFullBodyOverride.AnimGraphNode_StateMachine_2.CombatUpperBodyAdditive.AnimStateTransitionNode_1.Transition
    src=JumpLand
    via=AnimStateTransitionNode_1
   }
   {
    dst=Moving
    rule=VehicleFullBodyOverride.AnimGraphNode_StateMachine_2.CombatUpperBodyAdditive.AnimStateTransitionNode_9.Transition
    src=JumpLand
    via=AnimStateTransitionNode_9
   }
   {
    dst=JumpFall
    rule=VehicleFullBodyOverride.AnimGraphNode_StateMachine_2.CombatUpperBodyAdditive.AnimStateTransitionNode_7.Transition
    src=JumpStart
    via=AnimStateTransitionNode_7
   }
   {
    dst=Stop
    rule=VehicleFullBodyOverride.AnimGraphNode_StateMachine_2.CombatUpperBodyAdditive.AnimStateTransitionNode_10.Transition
    src=Moving
    via=AnimStateTransitionNode_10
   }
   {
    dst=Moving
    rule=VehicleFullBodyOverride.AnimGraphNode_StateMachine_2.CombatUpperBodyAdditive.AnimStateTransitionNode_2.Transition
    src=SA_ToMoving
    via=AnimStateTransitionNode_2
   }
   {
    dst=Swimming
    rule=VehicleFullBodyOverride.AnimGraphNode_StateMachine_2.CombatUpperBodyAdditive.AnimStateTransitionNode_3.Transition
    src=SA_ToSwimming
    via=AnimStateTransitionNode_3
   }
   {
    dst=Idle
    rule=VehicleFullBodyOverride.AnimGraphNode_StateMachine_2.CombatUpperBodyAdditive.AnimStateTransitionNode_18.Transition
    src=Stop
    via=AnimStateTransitionNode_18
   }
   {
    dst=Moving
    rule=VehicleFullBodyOverride.AnimGraphNode_StateMachine_2.CombatUpperBodyAdditive.AnimStateTransitionNode_12.Transition
    src=Stop
    via=AnimStateTransitionNode_12
   }
   {
    dst=Moving
    rule=VehicleFullBodyOverride.AnimGraphNode_StateMachine_2.CombatUpperBodyAdditive.AnimStateTransitionNode_5.Transition
    src=Swimming
    via=AnimStateTransitionNode_5
   }
  ]
 }
 Moving {
  path=VehicleFullBodyOverride.AnimGraphNode_StateMachine_2.CombatUpperBodyAdditive.AnimStateNode_0.Moving
  nodes [
   n1 = Get CurrentCurveValue() -> (CurrentCurveValue:double)
   Output Animation Pose(Result=Blend Poses by bool(
     BlendPose_0="Sequence Evaluator (by time) 'LabradorRiding_Player_Run_N_UpperBodyAdditive'"!(ExplicitTime:float=@n1) -> (Pose:Pose Link Structure) {sequence=&/LabradorRiding_Player_Run_N_UpperBodyAdditive}
     BlendPose_1="Sequence Evaluator (by time) 'LabradorRiding_Player_Walk_N_UpperBodyAdditive'"!(ExplicitTime:float=@n1) -> (Pose:Pose Link Structure) {sequence=&/LabradorRiding_Player_Walk_N_UpperBodyAdditive}
     BlendTime_0=0.5
     BlendTime_1=0.7
     bActiveValue=Get bIsRunning() -> (bIsRunning:bool)
    ) {blendType=Linear, transitionType=Inertialization})
  ]
 }
 Swimming {path=VehicleFullBodyOverride.AnimGraphNode_StateMachine_2.CombatUpperBodyAdditive.AnimStateNode_2.Swimming, nodes [Output Animation Pose()]}
 Idle {path=VehicleFullBodyOverride.AnimGraphNode_StateMachine_2.CombatUpperBodyAdditive.AnimStateNode_4.Idle, nodes [Output Animation Pose()]}
 JumpStart {
  path=VehicleFullBodyOverride.AnimGraphNode_StateMachine_2.CombatUpperBodyAdditive.AnimStateNode_5.JumpStart
  nodes [
   n1 = "Sequence Player (additive) 'LabradorRiding_HardSaddleWolf_Player_CombatJumpStart_Add'"() {bLoopAnimation=false, sequence=/LabradorRidingContent/Animation/HardSaddle/Wolf/LabradorRiding_HardSaddleWolf_Player_CombatJumpStart_Add}
   Output Animation Pose(Result=@n1)
  ]
 }
 JumpFall {
  path=VehicleFullBodyOverride.AnimGraphNode_StateMachine_2.CombatUpperBodyAdditive.AnimStateNode_6.JumpFall
  nodes [Output Animation Pose(Result=Blend Poses by bool(
     BlendPose_0="Sequence Player (additive) 'LabradorRiding_HardSaddleWolf_plyr_SlopeSliding_AddUpper'"() {sequence=/LabradorRidingContent/Animation/HardSaddle/Wolf/SlopeSliding/LabradorRiding_HardSaddleWolf_plyr_SlopeSliding_AddUpper}
     BlendPose_1="Sequence Player (additive) 'LabradorRiding_HardSaddleWolf_Player_CombatJumpLoop_Add'"() {sequence=/LabradorRidingContent/Animation/HardSaddle/Wolf/LabradorRiding_HardSaddleWolf_Player_CombatJumpLoop_Add}
     bActiveValue=Get bIsSlopeSliding()
    ) {blendType=Linear})]
 }
 Stop {
  path=VehicleFullBodyOverride.AnimGraphNode_StateMachine_2.CombatUpperBodyAdditive.AnimStateNode_7.Stop
  nodes [
   AnimGraphNode_StateMachine_0 = AdditiveStopTransition() {is=title AnimGraphNode_StateMachine, bAllowConduitEntryStates=true, bSkipFirstUpdateTransition=false, maxTransitionsPerFrame=1}
   Output Animation Pose(Result=@AnimGraphNode_StateMachine_0)
  ]
 }
 AdditiveStopTransition {
  entry=Conduit
  path=VehicleFullBodyOverride.AnimGraphNode_StateMachine_2.CombatUpperBodyAdditive.AnimStateNode_7.Stop.AnimGraphNode_StateMachine_0.AdditiveStopTransition
  ids=id
  nodes [
   AnimStateConduitNode_0 = Conduit()
   AnimStateNode_0 = FeetInAir() {is=title AnimStateNode}
   AnimStateNode_1 = FrontFeetPlanted() {is=title AnimStateNode}
   AnimStateNode_2 = BackFeetPlanted() {is=title AnimStateNode}
   AnimStateNode_3 = Backwards() {is=title AnimStateNode}
   AnimStateTransitionNode_1 = AnimStateTransitionNode() {logicType=TLT_Inertialization, priorityOrder=0}
   AnimStateTransitionNode_3 = AnimStateTransitionNode() {logicType=TLT_Inertialization}
   AnimStateTransitionNode_4 = AnimStateTransitionNode() {logicType=TLT_Inertialization, priorityOrder=2}
   AnimStateTransitionNode_5 = AnimStateTransitionNode() {logicType=TLT_Inertialization, priorityOrder=100}
  ]
  links [
   {
    dst=BackFeetPlanted
    rule=VehicleFullBodyOverride.AnimGraphNode_StateMachine_2.CombatUpperBodyAdditive.AnimStateNode_7.Stop.AnimGraphNode_StateMachine_0.AdditiveStopTransition.AnimStateTransitionNode_5.Transition
    src=Conduit
    via=AnimStateTransitionNode_5
   }
   {
    dst=Backwards
    rule=VehicleFullBodyOverride.AnimGraphNode_StateMachine_2.CombatUpperBodyAdditive.AnimStateNode_7.Stop.AnimGraphNode_StateMachine_0.AdditiveStopTransition.AnimStateTransitionNode_1.Transition
    src=Conduit
    via=AnimStateTransitionNode_1
   }
   {
    dst=FeetInAir
    rule=VehicleFullBodyOverride.AnimGraphNode_StateMachine_2.CombatUpperBodyAdditive.AnimStateNode_7.Stop.AnimGraphNode_StateMachine_0.AdditiveStopTransition.AnimStateTransitionNode_3.Transition
    src=Conduit
    via=AnimStateTransitionNode_3
   }
   {
    dst=FrontFeetPlanted
    rule=VehicleFullBodyOverride.AnimGraphNode_StateMachine_2.CombatUpperBodyAdditive.AnimStateNode_7.Stop.AnimGraphNode_StateMachine_0.AdditiveStopTransition.AnimStateTransitionNode_4.Transition
    src=Conduit
    via=AnimStateTransitionNode_4
   }
  ]
 }
 Conduit {
  path=VehicleFullBodyOverride.AnimGraphNode_StateMachine_2.CombatUpperBodyAdditive.AnimStateNode_7.Stop.AnimGraphNode_StateMachine_0.AdditiveStopTransition.AnimStateConduitNode_0.Conduit
  nodes [Result(bCanEnterTransition=true)]
 }
 FeetInAir {
  path=VehicleFullBodyOverride.AnimGraphNode_StateMachine_2.CombatUpperBodyAdditive.AnimStateNode_7.Stop.AnimGraphNode_StateMachine_0.AdditiveStopTransition.AnimStateNode_0.FeetInAir
  nodes [
   n1 = "Sequence Player (additive) 'LabradorRiding_HardSaddleWolf_Player_Stop_FeetInAir_Combat_Additive'"() {
    bLoopAnimation=false
    sequence=/LabradorRidingContent/Animation/HardSaddle/Wolf/LabradorRiding_HardSaddleWolf_Player_Stop_FeetInAir_Combat_Additive
   }
   Output Animation Pose(Result=@n1)
  ]
 }
 FrontFeetPlanted {
  path=VehicleFullBodyOverride.AnimGraphNode_StateMachine_2.CombatUpperBodyAdditive.AnimStateNode_7.Stop.AnimGraphNode_StateMachine_0.AdditiveStopTransition.AnimStateNode_1.FrontFeetPlanted
  nodes [
   n1 = "Sequence Player (additive) 'LabradorRiding_HardSaddleWolf_Player_Stop_FrontFeetPlanted_Combat_Additive'"() {
    bLoopAnimation=false
    sequence=/LabradorRidingContent/Animation/HardSaddle/Wolf/LabradorRiding_HardSaddleWolf_Player_Stop_FrontFeetPlanted_Combat_Additive
   }
   Output Animation Pose(Result=@n1)
  ]
 }
 BackFeetPlanted {
  path=VehicleFullBodyOverride.AnimGraphNode_StateMachine_2.CombatUpperBodyAdditive.AnimStateNode_7.Stop.AnimGraphNode_StateMachine_0.AdditiveStopTransition.AnimStateNode_2.BackFeetPlanted
  nodes [
   n1 = "Sequence Player (additive) 'LabradorRiding_HardSaddleWolf_Player_Stop_BackFeetPlanted_Combat_Additive'"() {
    bLoopAnimation=false
    sequence=/LabradorRidingContent/Animation/HardSaddle/Wolf/LabradorRiding_HardSaddleWolf_Player_Stop_BackFeetPlanted_Combat_Additive
   }
   Output Animation Pose(Result=@n1)
  ]
 }
 Backwards {
  path=VehicleFullBodyOverride.AnimGraphNode_StateMachine_2.CombatUpperBodyAdditive.AnimStateNode_7.Stop.AnimGraphNode_StateMachine_0.AdditiveStopTransition.AnimStateNode_3.Backwards
  nodes [
   AnimGraphNode_StateMachine_0 = AdditiveStopBackwards() {is=title AnimGraphNode_StateMachine, bAllowConduitEntryStates=true, bSkipFirstUpdateTransition=false, maxTransitionsPerFrame=1}
   Output Animation Pose(Result=@AnimGraphNode_StateMachine_0)
  ]
 }
 AdditiveStopBackwards {
  entry=Conduit
  path=VehicleFullBodyOverride.AnimGraphNode_StateMachine_2.CombatUpperBodyAdditive.AnimStateNode_7.Stop.AnimGraphNode_StateMachine_0.AdditiveStopTransition.AnimStateNode_3.Backwards.AnimGraphNode_StateMachine_0.AdditiveStopBackwards
  ids=id
  nodes [
   AnimStateConduitNode_0 = Conduit()
   AnimStateNode_0 = LeftBackFeetForward() {is=title AnimStateNode}
   AnimStateNode_1 = RightBackFeetForward() {is=title AnimStateNode}
   AnimStateNode_2 = LeftPlantedRightPass() {is=title AnimStateNode}
   AnimStateNode_3 = RightPlantedLeftPass() {is=title AnimStateNode}
   AnimStateTransitionNode_0 = AnimStateTransitionNode() {logicType=TLT_Inertialization}
   AnimStateTransitionNode_1 = AnimStateTransitionNode() {logicType=TLT_Inertialization, priorityOrder=2}
   AnimStateTransitionNode_2 = AnimStateTransitionNode() {logicType=TLT_Inertialization, priorityOrder=3}
   AnimStateTransitionNode_3 = AnimStateTransitionNode() {logicType=TLT_Inertialization, priorityOrder=100}
  ]
  links [
   {
    dst=LeftBackFeetForward
    rule=VehicleFullBodyOverride.AnimGraphNode_StateMachine_2.CombatUpperBodyAdditive.AnimStateNode_7.Stop.AnimGraphNode_StateMachine_0.AdditiveStopTransition.AnimStateNode_3.Backwards.AnimGraphNode_StateMachine_0.AdditiveStopBackwards.AnimStateTransitionNode_0.AnimationTransitionGraph_1
    src=Conduit
    via=AnimStateTransitionNode_0
   }
   {
    dst=LeftPlantedRightPass
    rule=VehicleFullBodyOverride.AnimGraphNode_StateMachine_2.CombatUpperBodyAdditive.AnimStateNode_7.Stop.AnimGraphNode_StateMachine_0.AdditiveStopTransition.AnimStateNode_3.Backwards.AnimGraphNode_StateMachine_0.AdditiveStopBackwards.AnimStateTransitionNode_2.AnimationTransitionGraph_1
    src=Conduit
    via=AnimStateTransitionNode_2
   }
   {
    dst=RightBackFeetForward
    rule=VehicleFullBodyOverride.AnimGraphNode_StateMachine_2.CombatUpperBodyAdditive.AnimStateNode_7.Stop.AnimGraphNode_StateMachine_0.AdditiveStopTransition.AnimStateNode_3.Backwards.AnimGraphNode_StateMachine_0.AdditiveStopBackwards.AnimStateTransitionNode_1.AnimationTransitionGraph_1
    src=Conduit
    via=AnimStateTransitionNode_1
   }
   {
    dst=RightPlantedLeftPass
    rule=VehicleFullBodyOverride.AnimGraphNode_StateMachine_2.CombatUpperBodyAdditive.AnimStateNode_7.Stop.AnimGraphNode_StateMachine_0.AdditiveStopTransition.AnimStateNode_3.Backwards.AnimGraphNode_StateMachine_0.AdditiveStopBackwards.AnimStateTransitionNode_3.AnimationTransitionGraph_1
    src=Conduit
    via=AnimStateTransitionNode_3
   }
  ]
 }
 Conduit {
  path=VehicleFullBodyOverride.AnimGraphNode_StateMachine_2.CombatUpperBodyAdditive.AnimStateNode_7.Stop.AnimGraphNode_StateMachine_0.AdditiveStopTransition.AnimStateNode_3.Backwards.AnimGraphNode_StateMachine_0.AdditiveStopBackwards.AnimStateConduitNode_0.Conduit
  nodes [Result(bCanEnterTransition=true)]
 }
 LeftBackFeetForward {
  path=VehicleFullBodyOverride.AnimGraphNode_StateMachine_2.CombatUpperBodyAdditive.AnimStateNode_7.Stop.AnimGraphNode_StateMachine_0.AdditiveStopTransition.AnimStateNode_3.Backwards.AnimGraphNode_StateMachine_0.AdditiveStopBackwards.AnimStateNode_0.LeftBackFeetForward
  nodes [
   n1 = "Sequence Player (additive) 'LabradorRiding_HardSaddleWolf_Player_Stop_Backwards_Combat_Add'"() {
    bLoopAnimation=false
    sequence=/LabradorRidingContent/Animation/HardSaddle/Wolf/LabradorRiding_HardSaddleWolf_Player_Stop_Backwards_Combat_Add
   }
   Output Animation Pose(Result=@n1)
  ]
 }
 RightBackFeetForward {
  path=VehicleFullBodyOverride.AnimGraphNode_StateMachine_2.CombatUpperBodyAdditive.AnimStateNode_7.Stop.AnimGraphNode_StateMachine_0.AdditiveStopTransition.AnimStateNode_3.Backwards.AnimGraphNode_StateMachine_0.AdditiveStopBackwards.AnimStateNode_1.RightBackFeetForward
  nodes [Output Animation Pose(Result="Sequence Player (additive) 'LabradorRiding_HardSaddleWolf_Player_Stop_Backwards_Combat_Add_fr19'"() {sequence=/LabradorRidingContent/Animation/HardSaddle/Wolf/LabradorRiding_HardSaddleWolf_Player_Stop_Backwards_Combat_Add_fr19})]
 }
 LeftPlantedRightPass {
  path=VehicleFullBodyOverride.AnimGraphNode_StateMachine_2.CombatUpperBodyAdditive.AnimStateNode_7.Stop.AnimGraphNode_StateMachine_0.AdditiveStopTransition.AnimStateNode_3.Backwards.AnimGraphNode_StateMachine_0.AdditiveStopBackwards.AnimStateNode_2.LeftPlantedRightPass
  nodes [
   n1 = "Sequence Player (additive) 'LabradorRiding_HardSaddleWolf_Player_Stop_Backwards_Combat_Add_fr24'"() {
    bLoopAnimation=false
    sequence=/LabradorRidingContent/Animation/HardSaddle/Wolf/LabradorRiding_HardSaddleWolf_Player_Stop_Backwards_Combat_Add_fr24
   }
   Output Animation Pose(Result=@n1)
  ]
 }
 RightPlantedLeftPass {
  path=VehicleFullBodyOverride.AnimGraphNode_StateMachine_2.CombatUpperBodyAdditive.AnimStateNode_7.Stop.AnimGraphNode_StateMachine_0.AdditiveStopTransition.AnimStateNode_3.Backwards.AnimGraphNode_StateMachine_0.AdditiveStopBackwards.AnimStateNode_3.RightPlantedLeftPass
  nodes [
   n1 = "Sequence Player (additive) 'LabradorRiding_HardSaddleWolf_Player_Stop_Backwards_Combat_Add_fr7'"() {
    bLoopAnimation=false
    sequence=/LabradorRidingContent/Animation/HardSaddle/Wolf/LabradorRiding_HardSaddleWolf_Player_Stop_Backwards_Combat_Add_fr7
   }
   Output Animation Pose(Result=@n1)
  ]
 }
 AnimationTransitionGraph_1 {
  path=VehicleFullBodyOverride.AnimGraphNode_StateMachine_2.CombatUpperBodyAdditive.AnimStateNode_7.Stop.AnimGraphNode_StateMachine_0.AdditiveStopTransition.AnimStateNode_3.Backwards.AnimGraphNode_StateMachine_0.AdditiveStopBackwards.AnimStateTransitionNode_0.AnimationTransitionGraph_1
  nodes [Result()]
 }
 AnimationTransitionGraph_1 {
  path=VehicleFullBodyOverride.AnimGraphNode_StateMachine_2.CombatUpperBodyAdditive.AnimStateNode_7.Stop.AnimGraphNode_StateMachine_0.AdditiveStopTransition.AnimStateNode_3.Backwards.AnimGraphNode_StateMachine_0.AdditiveStopBackwards.AnimStateTransitionNode_1.AnimationTransitionGraph_1
  nodes [Result()]
 }
 AnimationTransitionGraph_1 {
  path=VehicleFullBodyOverride.AnimGraphNode_StateMachine_2.CombatUpperBodyAdditive.AnimStateNode_7.Stop.AnimGraphNode_StateMachine_0.AdditiveStopTransition.AnimStateNode_3.Backwards.AnimGraphNode_StateMachine_0.AdditiveStopBackwards.AnimStateTransitionNode_2.AnimationTransitionGraph_1
  nodes [Result()]
 }
 AnimationTransitionGraph_1 {
  path=VehicleFullBodyOverride.AnimGraphNode_StateMachine_2.CombatUpperBodyAdditive.AnimStateNode_7.Stop.AnimGraphNode_StateMachine_0.AdditiveStopTransition.AnimStateNode_3.Backwards.AnimGraphNode_StateMachine_0.AdditiveStopBackwards.AnimStateTransitionNode_3.AnimationTransitionGraph_1
  nodes [Result(bCanEnterTransition=true)]
 }
 Transition {
  path=VehicleFullBodyOverride.AnimGraphNode_StateMachine_2.CombatUpperBodyAdditive.AnimStateNode_7.Stop.AnimGraphNode_StateMachine_0.AdditiveStopTransition.AnimStateTransitionNode_1.Transition
  nodes [Result(bCanEnterTransition=Get bShouldPlayStopBackwards() -> (bShouldPlayStopBackwards:bool))]
 }
 Transition {
  path=VehicleFullBodyOverride.AnimGraphNode_StateMachine_2.CombatUpperBodyAdditive.AnimStateNode_7.Stop.AnimGraphNode_StateMachine_0.AdditiveStopTransition.AnimStateTransitionNode_3.Transition
  nodes [Result()]
 }
 Transition {
  path=VehicleFullBodyOverride.AnimGraphNode_StateMachine_2.CombatUpperBodyAdditive.AnimStateNode_7.Stop.AnimGraphNode_StateMachine_0.AdditiveStopTransition.AnimStateTransitionNode_4.Transition
  nodes [Result()]
 }
 Transition {
  path=VehicleFullBodyOverride.AnimGraphNode_StateMachine_2.CombatUpperBodyAdditive.AnimStateNode_7.Stop.AnimGraphNode_StateMachine_0.AdditiveStopTransition.AnimStateTransitionNode_5.Transition
  nodes [Result(bCanEnterTransition=true)]
 }
 JumpLand {
  path=VehicleFullBodyOverride.AnimGraphNode_StateMachine_2.CombatUpperBodyAdditive.AnimStateNode_8.JumpLand
  nodes [
   n1 = "Sequence Player (additive) 'LabradorRiding_HardSaddleWolf_Player_CombatJumpLand_Add'"() {bLoopAnimation=false, sequence=/LabradorRidingContent/Animation/HardSaddle/Wolf/LabradorRiding_HardSaddleWolf_Player_CombatJumpLand_Add}
   Output Animation Pose(Result=@n1)
  ]
 }
 Transition {path=VehicleFullBodyOverride.AnimGraphNode_StateMachine_2.CombatUpperBodyAdditive.AnimStateTransitionNode_1.Transition, nodes [Result()]}
 Transition {
  path=VehicleFullBodyOverride.AnimGraphNode_StateMachine_2.CombatUpperBodyAdditive.AnimStateTransitionNode_10.Transition
  nodes [Result(bCanEnterTransition=Get bMovementToIdle() -> (bMovementToIdle:bool))]
 }
 Transition {
  path=VehicleFullBodyOverride.AnimGraphNode_StateMachine_2.CombatUpperBodyAdditive.AnimStateTransitionNode_12.Transition
  nodes [Result(bCanEnterTransition=Get bStateRuleIdleMoving() -> (bStateRuleIdleMoving:bool))]
 }
 Transition {
  path=VehicleFullBodyOverride.AnimGraphNode_StateMachine_2.CombatUpperBodyAdditive.AnimStateTransitionNode_17.Transition
  nodes [Result(bCanEnterTransition=Get bIsFalling_BP())]
 }
 Transition {path=VehicleFullBodyOverride.AnimGraphNode_StateMachine_2.CombatUpperBodyAdditive.AnimStateTransitionNode_18.Transition, nodes [Result()]}
 Transition {
  path=VehicleFullBodyOverride.AnimGraphNode_StateMachine_2.CombatUpperBodyAdditive.AnimStateTransitionNode_2.Transition
  nodes [Result(bCanEnterTransition=Get bStateRuleIdleMoving())]
 }
 Transition {
  path=VehicleFullBodyOverride.AnimGraphNode_StateMachine_2.CombatUpperBodyAdditive.AnimStateTransitionNode_3.Transition
  nodes [Result(bCanEnterTransition=Get bIsSurfaceSwimming())]
 }
 Transition {
  path=VehicleFullBodyOverride.AnimGraphNode_StateMachine_2.CombatUpperBodyAdditive.AnimStateTransitionNode_4.Transition
  nodes [Result(bCanEnterTransition=Get bIsJumping_BP())]
 }
 Transition {
  path=VehicleFullBodyOverride.AnimGraphNode_StateMachine_2.CombatUpperBodyAdditive.AnimStateTransitionNode_5.Transition
  nodes [Result(bCanEnterTransition=NOT Boolean(A=Get bIsSurfaceSwimming()))]
 }
 Transition {
  path=VehicleFullBodyOverride.AnimGraphNode_StateMachine_2.CombatUpperBodyAdditive.AnimStateTransitionNode_6.Transition
  nodes [Result(bCanEnterTransition=Get bStateRuleFallIdle() -> (bStateRuleFallIdle:bool))]
 }
 Transition {path=VehicleFullBodyOverride.AnimGraphNode_StateMachine_2.CombatUpperBodyAdditive.AnimStateTransitionNode_7.Transition, nodes [Result()]}
 Transition {
  path=VehicleFullBodyOverride.AnimGraphNode_StateMachine_2.CombatUpperBodyAdditive.AnimStateTransitionNode_9.Transition
  nodes [Result(bCanEnterTransition=Get bStateRuleJumpLandMoving() -> (bStateRuleJumpLandMoving:bool))]
 }
 MeleeOffset {
  entry=MeleeUpperBodyOffset
  path=VehicleFullBodyOverride.AnimGraphNode_StateMachine_3.MeleeOffset
  ids=id
  nodes [AnimStateNode_0 = MeleeUpperBodyOffset() {is=title AnimStateNode}]
 }
 MeleeUpperBodyOffset {
  path=VehicleFullBodyOverride.AnimGraphNode_StateMachine_3.MeleeOffset.AnimStateNode_0.MeleeUpperBodyOffset
  nodes [
   n1 = Get BodyYaw() -> (BodyYaw:float)
   n2 = Blend Poses by bool(
    BlendPose_0=Blendspace Player 'LabradorRiding_HardSaddleWolf_MeleeOffset_1H2Swing_BS'!(X=@n1) -> (Pose) {blendSpace=/LabradorRidingContent/Animation/HardSaddle/Wolf/MeleeOffset/LabradorRiding_HardSaddleWolf_MeleeOffset_1H2Swing_BS}
    BlendPose_1=Blend Poses by bool(
     BlendPose_0=Blendspace Player 'LabradorRiding_HardSaddleWolf_MeleeOffset_2HSword_BS'!(X=@n1) -> (Pose) {blendSpace=/LabradorRidingContent/Animation/HardSaddle/Wolf/MeleeOffset/LabradorRiding_HardSaddleWolf_MeleeOffset_2HSword_BS}
     BlendPose_1=Additive Identity Pose()
     bActiveValue=Get bIsTwoHandedSword() -> (bIsTwoHandedSword:bool)
    ) {blendType=Linear}
    bActiveValue=Get bIsOneHandedDualWield() -> (bIsOneHandedDualWield:bool)
   ) {blendType=Linear}
   n3 = Blendspace Player 'LabradorRiding_HardSaddleWolf_MeleeOffset_DW_BS'!(X=@n1) -> (Pose) {blendSpace=/LabradorRidingContent/Animation/HardSaddle/Wolf/MeleeOffset/LabradorRiding_HardSaddleWolf_MeleeOffset_DW_BS}
   n4 = Get bIsDualWield() -> (bIsDualWield:bool)
   n5 = Get BodyYaw()
   Output Animation Pose(Result=Apply Additive(
     Base=Blend Poses by bool(BlendPose_0=@n3, BlendPose_1=@n2, bActiveValue=@n4) {blendType=Linear}
     Additive=Blendspace Player 'LabradorRiding_HardSaddleBoar_Player_MeleeLegCorrection_BS'!(X=@n5) -> (Pose) {blendSpace=/LabradorRidingContent/Animation/HardSaddle/Boar/LabradorRiding_HardSaddleBoar_Player_MeleeLegCorrection_BS}
    ) {alphaBoolBlend {blendOption=Linear}})
  ]
 }
 EntryAdditive {
  entry=AddPose
  path=VehicleFullBodyOverride.AnimGraphNode_StateMachine_4.EntryAdditive
  ids=id
  nodes [
   AnimStateNode_0 = AddPose() {is=title AnimStateNode}
   AnimStateNode_1 = EntryAdd() {is=title AnimStateNode}
   AnimStateTransitionNode_0 = AnimStateTransitionNode() {crossfadeDuration=0.1, logicType=TLT_Inertialization}
   AnimStateTransitionNode_1 = AnimStateTransitionNode() {bAutomaticRuleBasedOnSequencePlayerInState=true, crossfadeDuration=0.3, logicType=TLT_Inertialization}
  ]
  links [
   {
    dst=EntryAdd
    rule=VehicleFullBodyOverride.AnimGraphNode_StateMachine_4.EntryAdditive.AnimStateTransitionNode_0.Transition
    src=AddPose
    via=AnimStateTransitionNode_0
   }
   {
    dst=AddPose
    rule=VehicleFullBodyOverride.AnimGraphNode_StateMachine_4.EntryAdditive.AnimStateTransitionNode_1.Transition
    src=EntryAdd
    via=AnimStateTransitionNode_1
   }
  ]
 }
 AddPose {
  path=VehicleFullBodyOverride.AnimGraphNode_StateMachine_4.EntryAdditive.AnimStateNode_0.AddPose
  nodes [Output Animation Pose(Result=Additive Identity Pose())]
 }
 EntryAdd {
  path=VehicleFullBodyOverride.AnimGraphNode_StateMachine_4.EntryAdditive.AnimStateNode_1.EntryAdd
  nodes [Output Animation Pose(Result="Sequence Player (additive) 'LabradorRiding_HardSaddleWolf_Player_SimpleEntry'"() {bLoopAnimation=false, sequence=/LabradorRidingContent/Animation/HardSaddle/Wolf/LabradorRiding_HardSaddleWolf_Player_SimpleEntry})]
 }
 Transition {
  path=VehicleFullBodyOverride.AnimGraphNode_StateMachine_4.EntryAdditive.AnimStateTransitionNode_0.Transition
  nodes [Result(bCanEnterTransition=Get bJustEntered() -> (bJustEntered:bool))]
 }
 Transition {path=VehicleFullBodyOverride.AnimGraphNode_StateMachine_4.EntryAdditive.AnimStateTransitionNode_1.Transition, nodes [Result()]}
 VehicleLowerBodyOverride {nodes [
   Input Pose() {name=InPoseLowerBodyDefault}
   Input Pose() {name=InPoseUpperBody}
   Output Pose() {layerGroup=VehicleOverride}
  ]}
 VehicleSplitBodyOverride {nodes [
   Input Pose() {name=InPoseSplitBody}
   Input Pose() {name=InPoseUpperAndLowerBody}
   Input Pose() {name=InPoseUpperAndLowerBodyRemoveRoot}
   Output Pose() {layerGroup=Vehicle Override}
  ]}
 WeaponAttachmentProceduralAdjustment {nodes [
   n1 = "Transform (Modify) Bone - Bone: pelvis"!(ComponentPose=Local To Component(LocalPose=Input Pose()), Alpha=0.25, Rotation=Get MeleeTwistRot() -> (MeleeTwistRot:Rotator)) -> (Pose) {
    rotationMode=BMM_Additive
    rotationSpace=BCS_WorldSpace
    translationSpace=BCS_WorldSpace
    alphaBoolBlend {blendOption=Linear}
    boneToModify {boneName=pelvis}
   }
   n2 = "Transform (Modify) Bone - Bone: weapon_r"!(ComponentPose=@n1, bAlphaBoolEnabled:bool=True, Scale="0.001,0.001,0.001") -> (Pose) {alphaInputType=Bool, scaleMode=BMM_Replace, alphaBoolBlend {blendOption=Linear}, boneToModify {boneName=weapon_r}, scale=<0.001 0.001 0.001>}
   n3 = "Transform (Modify) Bone - Bone: weapon_l"!(ComponentPose=@n2, bAlphaBoolEnabled:bool=True, Scale="0.001,0.001,0.001") -> (Pose) {alphaInputType=Bool, scaleMode=BMM_Replace, alphaBoolBlend {blendOption=Linear}, boneToModify {boneName=weapon_l}, scale=<0.001 0.001 0.001>}
   n4 = "Transform (Modify) Bone - Bone: hand_attach_r"!(ComponentPose=@n3, bAlphaBoolEnabled:bool=True, Scale="0.001,0.001,0.001") -> (Pose) {
    alphaInputType=Bool
    scaleMode=BMM_Replace
    alphaBoolBlend {blendOption=Linear}
    boneToModify {boneName=hand_attach_r}
    scale=<0.001 0.001 0.001>
   }
   n5 = "Transform (Modify) Bone - Bone: hand_attach_l"!(ComponentPose=@n4, bAlphaBoolEnabled:bool=True, Scale="0.001,0.001,0.001") -> (Pose) {
    alphaInputType=Bool
    scaleMode=BMM_Replace
    alphaBoolBlend {blendOption=Linear}
    boneToModify {boneName=hand_attach_l}
    scale=<0.001 0.001 0.001>
   }
   Output Pose(Result=Component To Local(ComponentPose=@n5))
  ]}
]
variables [DisableLeanWeight {type=number}]
